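# Sound cue for the Masako pickaxe swing, a cosmetic melee weapon sound.
# Each quality-level branch mixes close-attenuated Masako swings with medium-attenuated generic distant fire axe swings, using pitch or volume modulators.
# Lower quality levels draw on fewer random variations, and the last two levels share the smallest branch, which has a single distant wave.
# Overridden concurrency limits it per owner, stopping the oldest instance.
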SoundCue /BRCosmetics/Sounds/PickAxes/Masako/PA_Masako_Swing
root=QualityLevel_1
nodes [
 {id=Attenuation_10, children [Modulator_6], properties {attenuationSettings=/Game/Sounds/Attenuation/Fort_Weapon_Pickaxe_Close_Att}}
 {id=Attenuation_11, children [Modulator_7], properties {attenuationSettings=/Game/Sounds/Attenuation/Fort_Weapon_Pickaxe_Medium_Att}}
 {id=Attenuation_6, children [Random_8], properties {attenuationSettings=&Fort_Weapon_Pickaxe_Close_Att}}
 {id=Attenuation_7, children [Random_0], properties {attenuationSettings=&Fort_Weapon_Pickaxe_Medium_Att}}
 {id=Attenuation_8, children [Random_9], properties {attenuationSettings=&Fort_Weapon_Pickaxe_Close_Att}}
 {id=Attenuation_9, children [Random_4], properties {attenuationSettings=&Fort_Weapon_Pickaxe_Medium_Att}}
 {id=Mixer_3, children [Attenuation_6, Attenuation_7], properties {inputVolume [1,1]}}
 {id=Mixer_4, children [Attenuation_8, Attenuation_9], properties {inputVolume [1,1]}}
 {id=Mixer_5, children [Attenuation_10, Attenuation_11], properties {inputVolume [1,1]}}
 {id=Modulator_0, children [Mixer_3], properties {pitchMax=1.03, pitchMin=0.97}}
 {id=Modulator_5, children [Mixer_4]}
 {id=Modulator_6, children [Random_5], properties {pitchMax=1.2, pitchMin=0.9}}
 {id=Modulator_7, children [WavePlayer_23], properties {pitchMax=1.2, pitchMin=0.9}}
 {id=QualityLevel_1, children [Modulator_0,Modulator_5,Mixer_5,Mixer_5]}
 {id=Random_0, children [WavePlayer_0,WavePlayer_1,WavePlayer_2,WavePlayer_3,WavePlayer_4,WavePlayer_5], properties {weights [1,1,1,1,1,1]}}
 {id=Random_4, children [WavePlayer_21, WavePlayer_9, WavePlayer_22], properties {weights [1,1,1]}}
 {id=Random_5, children [WavePlayer_40, WavePlayer_41], properties {weights [1,1]}}
 {
  id=Random_8
  children [WavePlayer_24, WavePlayer_25, WavePlayer_26, WavePlayer_27, WavePlayer_29, WavePlayer_30]
  properties {weights [1,1,1,1,1,1]}
 }
 {id=Random_9, children [WavePlayer_37, WavePlayer_38, WavePlayer_39], properties {weights [1,1,1]}}
 {id=WavePlayer_0, properties {soundWaveAssetPtr=/Game/Athena/Sounds/Weapons/Fire_Axe_Swing_Distant_04}}
 {id=WavePlayer_1, properties {soundWaveAssetPtr=/Game/Athena/Sounds/Weapons/Fire_Axe_Swing_Distant_01}}
 {id=WavePlayer_2, properties {soundWaveAssetPtr=/Game/Athena/Sounds/Weapons/Fire_Axe_Swing_Distant_02}}
 {id=WavePlayer_21, properties {soundWaveAssetPtr=&Fire_Axe_Swing_Distant_04}}
 {id=WavePlayer_22, properties {soundWaveAssetPtr=/Game/Athena/Sounds/Weapons/Fire_Axe_Swing_Distant_06}}
 {id=WavePlayer_23, properties {soundWaveAssetPtr=&Fire_Axe_Swing_Distant_06}}
 {id=WavePlayer_24, properties {soundWaveAssetPtr=&/PA_Masako_Swing_01}}
 {id=WavePlayer_25, properties {soundWaveAssetPtr=&/PA_Masako_Swing_02}}
 {id=WavePlayer_26, properties {soundWaveAssetPtr=&/PA_Masako_Swing_03}}
 {id=WavePlayer_27, properties {soundWaveAssetPtr=&/PA_Masako_Swing_04}}
 {id=WavePlayer_29, properties {soundWaveAssetPtr=&/PA_Masako_Swing_05}}
 {id=WavePlayer_3, properties {soundWaveAssetPtr=/Game/Athena/Sounds/Weapons/Fire_Axe_Swing_Distant_03}}
 {id=WavePlayer_30, properties {soundWaveAssetPtr=&/PA_Masako_Swing_06}}
 {id=WavePlayer_37, properties {soundWaveAssetPtr=&PA_Masako_Swing_01}}
 {id=WavePlayer_38, properties {soundWaveAssetPtr=&PA_Masako_Swing_02}}
 {id=WavePlayer_39, properties {soundWaveAssetPtr=&PA_Masako_Swing_03}}
 {id=WavePlayer_4, properties {soundWaveAssetPtr=/Game/Athena/Sounds/Weapons/Fire_Axe_Swing_Distant_05}}
 {id=WavePlayer_40, properties {soundWaveAssetPtr=&PA_Masako_Swing_01}}
 {id=WavePlayer_41, properties {soundWaveAssetPtr=&PA_Masako_Swing_02}}
 {id=WavePlayer_5, properties {soundWaveAssetPtr=&Fire_Axe_Swing_Distant_06}}
 {id=WavePlayer_9, properties {soundWaveAssetPtr=&Fire_Axe_Swing_Distant_03}}
]
properties {
 audioPropertiesSheet=/BRCosmetics/Sounds/Settings/PropertySheets/Pickaxe/APS_Cosmetic_Pickaxe_Swing
 bOverrideConcurrency=true
 bPrimeOnLoad=true
 maxDistance=3000
 soundClassObject=&Fort_WeaponMelee_PSM
 volumeMultiplier=0.2
 concurrencyOverrides {bLimitToOwner=true, maxCount=2, resolutionRule=StopOldest}
 concurrencySet [&Player_Melee_Weapon_Conc]
}
